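# Niagara idle cosmetic effect for the OakWeld character that sheds warm-colored leaf sprites from skeletal sockets.
# One emitter partitions particles across sockets, with spawn rate scaled by system reactivity and gated by owner speed, while land and sprint values scale launch velocity and lifetime.
# Leaves inherit owner motion, then tumble under curl noise, gravity and drag using sprites aligned to mesh orientation.
# Time-of-day and front-end modifiers adjust brightness, and camera distance scales size.
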NiagaraSystem /BRCosmetics/Characters/Player/Female/Medium/Bodies/F_MED_OakWeld/FX/NS_Character_OakWeld_Idle
dependencies {
 usedDataInterfaces [
  &NiagaraDataInterfaceArrayColor
  &NiagaraDataInterfaceArrayFloat3
  &NiagaraDataInterfaceCurve
  &NiagaraDataInterfaceCamera
  &NiagaraDataInterfaceSpriteRendererInfo
  &NiagaraDataInterfaceSocketReader
 ]
 usedDynamicInputs [
  &Subtract_Float
  &ClampFloat
  &Divide_Float
  &Add_Float
  &FNTimeOfDayColor
  &Max_Float
 ]
 usedModules [
  &self:SystemSpawnScript.SystemScriptSource.SystemScriptGraph.NiagaraNodeAssignment_1.SetVariables_#1
  &SystemState
  &Timeline
  &self:SystemSpawnScript.SystemScriptSource.SystemScriptGraph.NiagaraNodeAssignment_2.SetVariables_#2
  &self:SystemSpawnScript.SystemScriptSource.SystemScriptGraph.NiagaraNodeAssignment_0.SetVariables_#3
  &self:SystemSpawnScript.SystemScriptSource.SystemScriptGraph.NiagaraNodeAssignment_4.SetVariables_#4
  &self:SystemSpawnScript.SystemScriptSource.SystemScriptGraph.NiagaraNodeAssignment_5.SetVariables_#5
 ]
}
emitters {Leaves {
  emitterUpdateScript [
   &EmitterState()
   &SpawnRate(
    Spawn Probability=0.8
    SpawnRate=&Multiply_Float(
     A=&Multiply_Float(A=50, B=$System.ReactivitySpawn)
     B=&MakeCustomFloatFromBool(
      Bool=&SetBoolByFloatComparison(A=&VectorLength(Vector=$Engine.Owner.Velocity), B=1190)
      True Float=0
      False Float=1
     )
    )
    Use Spawn Probability=true
   )
  ]
  particleSpawnScript [
   &InitializeParticle(
    Color=&SelectLinearColorFromArray(
     Array Sampling Mode=::NewEnumerator0
     Direct Array Index=?
     Color Selection Array=DataInterface() {propertyValues {ColorData [<rgba 0.75 0.275 0.0375 1>, <rgba 1 0.2087 0 1>, <rgba 1 0.4458 0.05 1>, <rgba 0.75 0.5125 0.0375 1>]}}
     Random Seed=0
    )
    Lifetime=&Multiply_Float(
     A=&RandomRangeFloat(Minimum=0.8, Maximum=1)
     B=&Lerp_Float(A=1.5, B=1, Alpha=$System.LandorSprint)
    )
    Sprite Rotation Mode=::NewEnumerator2
    Sprite Size Mode=::NewEnumerator2
    UsePositionOffset=false
    Write Material Random=true
   ) {module=InitializeParticle001}
   &PartitionParticles(Number Of Partitions=10, Partition Mode:ENiagara_PartitionMode_v2=::NewEnumerator5)
   &SocketLocation(
    Interpolate Socket Sampling=true
    Orientation Sampling=::NewEnumerator1
    SamplePosOffset=true
    Sampled Position Offset=&MergeHelper_VectorSwitch(
     0=&SelectVectorFromArray(
      Array Sampling Mode=::NewEnumerator0
      Direct Array Index=?
      Vector Selection Array=DataInterface() {propertyValues {FloatData [<0 0 -10>, <0 0 10>]}}
      Random Seed=0
     )
     1=&SelectVectorFromArray(
      Array Sampling Mode=::NewEnumerator0
      Direct Array Index=?
      Vector Selection Array=DataInterface() {propertyValues {FloatData [<0 0 -10>, <0 0 10>]}}
      Random Seed=0
     )
     2=&SelectVectorFromArray(
      Array Sampling Mode=::NewEnumerator0
      Direct Array Index=?
      Vector Selection Array=DataInterface() {propertyValues {FloatData [<0 0 -10>, <0 0 10>]}}
      Random Seed=0
     )
     3=&SelectVectorFromArray(
      Array Sampling Mode=::NewEnumerator0
      Direct Array Index=?
      Vector Selection Array=DataInterface() {propertyValues {FloatData [<0 0 -10>, <0 0 10>]}}
      Random Seed=0
     )
     4=<0 0 0>
     5=<0 0 0>
     6=<0 0 0>
     7=<0 0 0>
     8=<0 0 0>
     9=<0 0 0>
     Index=$Particles.PartitionParticles.Partition
     Num=::NewEnumerator8
    )
    Socket Index=$Particles.PartitionParticles.Partition
    Socket Reader=$System.Socket Reader
    Socket Sampling Mode=::NewEnumerator3
    VelocitySpeedLimit=false
   )
   &DynamicMaterialParameters(
    Index 0 Param 1=&FN_FrontEndFloatModifier(
     Gameplay=&FNTimeOfDayScalar(Night Scale=0.06, Morning Scale=0.06, Day Scale=0.1, Evening Scale=0.08)
     Front End=0.01
    )
    Index 0 Param 2=10
    Module.Param0WriteEnabled=true
    Module.Param1WriteEnabled=true
    Module.Param2WriteEnabled=true
    Module.Param3WriteEnabled=true
   ) {module=DynamicMaterialParameters001}
   &ShapeLocation(MaskSpawnGroup=false, Override Seed=false, Sphere Radius=2, Sphere Surface Distribution=1, UseOwnerScale=true)
   &AddVelocity(
    Cone Axis=<0 1 1>
    Module.UseConeFalloff=false
    Module.UseVelDistribution=false
    Override Seed=false
    Velocity Mode=::NewEnumerator2
    Velocity Speed=&Multiply_Float(A=&RandomRangeFloat(Minimum=25, Maximum=50), B=1)
   )
   &AddVelocity(
    Module.Use Falloff Distance=false
    Origin Offset=<0 -50 80>
    Override Seed=false
    Velocity Mode=::NewEnumerator1
    Velocity Speed=&Multiply_Float(
     A=&RandomRangeFloat(Minimum=100, Maximum=120)
     B=&Lerp_Float(A=0.1, B=1, Alpha=$System.LandReactivity)
    )
   ) {module=AddVelocity001}
   &CameraOffset(Camera Offset Amount=10)
   &InheritVelocity(
    Inherited Velocity Amount Scale=&MakeVector(X=0.65, Y=0.65, Z=0.1)
    Inherited Velocity Speed Limit=500
    Speed Limit=false
    UseSpeedThreshold=false
   )
   &InitialMeshOrientation(Use Rotation Vector=false)
  ]
  particleUpdateScript [
   &ParticleState()
   &ScaleSpriteSize(Uniform Curve Sprite Scale=DataInterface() {propertyValues {Curve {
       defaultValue=1
       keys [
        {arriveTangent=7.4671, interpMode=RCIM_Cubic, leaveTangent=7.4671, tangentMode=RCTM_User, time=0, value=0.5}
        {arriveTangent=0.0097, interpMode=RCIM_Cubic, leaveTangent=0.0097, tangentMode=RCTM_User, time=0.1, value=1}
        {interpMode=RCIM_Cubic, tangentMode=RCTM_User, time=0.5, value=1}
        {tangentMode=RCTM_User, time=1, value=0}
       ]
      }}})
   &ScaleSpriteSize(
    Initial Sprite Size:Vector2f=$Particles.SpriteSize
    Uniform Curve Index=/Game/Effects/Niagara/Modules/Math/FN_CameraDistance(
     0Dist:NiagaraFloat=70
     1Dist:NiagaraFloat=90
     Camera Query:NiagaraDataInterfaceCamera=DataInterface()
     Position:NiagaraPosition=$Particles.Position
    )
   ) {module=ScaleSpriteSize001}
   &UpdateMeshOrientation(
    Rotation Rate=&FloatFromCurve(
     FloatCurve=DataInterface() {propertyValues {Curve {
        defaultValue=1
        keys [{interpMode=RCIM_Cubic, tangentMode=RCTM_User, time=0, value=1}, {interpMode=RCIM_Cubic, tangentMode=RCTM_User, time=1, value=0}]
       }}}
     Scale Curve=&RandomRangeFloat(Minimum=0.5, Maximum=1.5)
    )
    Rotation Vector=&RandomVector(Vector Scale=1)
   )
   &SubUVAnimation(
    Random Start Frame=true
    "Sprite Renderer "=DataInterface() {propertyValues {SpriteRenderer=&self:Leaves_5.NiagaraSpriteRendererProperties_2}}
    SubUV Lookup Index Scale=0.5
    UseEndFrame=true
    UseStartFrame=true
   ) {module=SubUVAnimation002}
   /Niagara/Modules/Update/Forces/CurlNoiseForce(
    Noise Frequency=100
    Noise Strength=&Multiply_Float(
     A=100
     B=&FloatFromCurve(FloatCurve=DataInterface() {propertyValues {Curve {keys [{interpMode=RCIM_Cubic, tangentMode=RCTM_User, time=0, value=0}, {interpMode=RCIM_Cubic, tangentMode=RCTM_User, time=1, value=1}]}}})
    )
    Pan Noise=false
   )
   &GravityForce(Gravity=&FN_FrontEndVector3Modifier(
     Gameplay=&RandomRangeVector(Minimum=<0 0 -200>, Maximum=<0 0 -300>)
     Front End=&RandomRangeVector(Minimum=<0 0 -100>, Maximum=<0 0 -250>)
    ))
   &AerodynamicDrag(Aerodynamic Drag=0.5, Aerodynamic Rotational Drag=0.5, Use Linear Drag=true, Use Rotational Drag=true)
   &SolveForcesAndVelocity(Clamp Velocity=false, Limit Acceleration=false)
   &AlignSpriteToMeshOrientation(SetAlignment=true, SetFacing=true)
   &ScaleColor(
    Scale RGB=&VectorFromFloat(Value=&FN_FrontEndFloatModifier(
      Gameplay=&FNTimeOfDayScalar(Night Scale=0.1, Morning Scale=0.25, Day Scale=0.5, Evening Scale=0.3)
      Front End=0.31
     ))
    ScaleA=true
    ScaleRGB=true
   )
  ]
  rendererClasses [&NiagaraSpriteRendererProperties]
  renderers [{rendererClass=&NiagaraSpriteRendererProperties}]
 }}
summary {
 systemName=NS_Character_OakWeld_Idle
 emitters [{emitterName=Leaves, rendererClasses [&NiagaraSpriteRendererProperties]}]
 userVariables [
  User.bIsFrontEnd {type=NiagaraBool}
  User.bIsFrontEndPreview {type=NiagaraBool}
  User.bPawnSprintOrSlide {type=NiagaraBool}
  User.LandGameTime {type=NiagaraFloat, defaultValue=0, description=""}
  User.ReactivityDebug {type=NiagaraBool}
 ]
}
