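# Physics asset for the F_MED_OakWeld female body cosmetic, covering its dangling rope and left-thigh leaf bones.
# Front and back four-link rope chains branch from Mstr_rope on the pelvis with limited swing and twist, cross-linked by angular-free constraints at their last two links.
# A limited leaf joint hangs off the left thigh.
# Capsules on root, pelvis, thighs, right calf and right arm supply body shapes, joined by angular-free constraints.
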PhysicsAsset /BRCosmetics/Characters/Player/Female/Medium/Bodies/F_MED_OakWeld/Meshes/F_MED_OakWeld_Physics
bodies {
 Base_Leaves_l [{shapeType=Capsule, length=0.505, radius=0.505, shapeName=Base_Leaves_l_capsule, center=<-4.1914 0 0>, rotation=<rot 0 -90 0>}]
 Base_rope_bk [{shapeType=Capsule, length=0.505, radius=0.505, shapeName=Base_rope_bk_capsule, rotation=<rot 0 -90 0>}]
 Base_rope_fr [{shapeType=Capsule, length=0.505, radius=0.505, shapeName=Base_rope_fr_capsule, rotation=<rot 0 -90 0>}]
 Dyn_Leaves_l [
  {
   shapeType=Capsule
   length=9.0892
   radius=8.023
   shapeName=Dyn_Leaves_l_capsule
   center=<10.3399 1.9752 8.8188>
   rotation=<rot 0 7.5203 -88.1268>
  }
  {shapeType=Capsule, length=7, radius=7, shapeName=Dyn_Leaves_l_capsule, center=<9.1772 12.7969 -3.9172>, rotation=<rot 10 0.0001 0>}
 ]
 Dyn_rope_bk_01 [{
   shapeType=Capsule
   length=5.2535
   radius=3.0781
   shapeName=Dyn_rope_bk_01_capsule
   center=<4.0178 0 -1.5>
   rotation=<rot 85 179.9999 179.9997>
  }]
 Dyn_rope_bk_02 [{shapeType=Capsule, length=6.4002, radius=3.2581, shapeName=Dyn_rope_bk_02_capsule, center=<6.7128 0 -0.6>, rotation=<rot 85 0.0005 0.0002>}]
 Dyn_rope_bk_03 [{
   shapeType=Capsule
   length=21.8369
   radius=5.0412
   shapeName=Dyn_rope_bk_03_capsule
   center=<9.3 0.0001 2.5>
   rotation=<rot 75.0001 -179.9997 179.9998>
  }]
 Dyn_rope_bk_04 [{shapeType=Capsule, length=6.5, radius=7.005, shapeName=Dyn_rope_bk_04_capsule, center=<4 -5.5 -6>, rotation=<rot 9.9998 0.0007 -69.9998>}]
 Dyn_rope_fr_01 [{
   shapeType=Capsule
   length=7.0929
   radius=3.4903
   shapeName=Dyn_rope_fr_01_capsule
   center=<6.6951 -1.8331 -1.2895>
   rotation=<rot -0.2398 -95.6991 -90.895>
  }]
 Dyn_rope_fr_02 [{
   shapeType=Capsule
   length=7.2476
   radius=3.0467
   shapeName=Dyn_rope_fr_02_capsule
   center=<6.6061 1.2688 -0.7>
   rotation=<rot 0 -91.3498 -85.4773>
  }]
 Dyn_rope_fr_03 [{
   shapeType=Capsule
   length=12.5006
   radius=5.1849
   shapeName=Dyn_rope_fr_03_capsule
   center=<13.5202 2.4184 -2.6>
   rotation=<rot 18.8028 -102.7932 -104.2118>
  }]
 Dyn_rope_fr_04 [{
   shapeType=Capsule
   length=9.5
   radius=4.505
   shapeName=Dyn_rope_fr_04_capsule
   center=<4.3334 0.3872 -1.3941>
   rotation=<rot -0.8671 -94.924 80.0375>
  }]
 Mstr_rope [{shapeType=Capsule, length=0.505, radius=0.505, shapeName=Mstr_rope_capsule, rotation=<rot 0 -90 0>}]
 calf_r [{shapeType=Capsule, length=40.118, radius=6.4041, shapeName=calf_r_capsule, center=<29.4266 0 0>, rotation=<rot -90 -0.0001 0>}]
 lowerarm_r [{shapeType=Capsule, length=25, radius=3.5, shapeName=lowerarm_r_capsule, center=<-12.5353 0 -4.7127>, rotation=<rot -0.0001 -89.9999 94.9999>}]
 pelvis [{shapeType=Capsule, length=10.5146, radius=8.4117, shapeName=pelvis_capsule, rotation=<rot 12.9122 -90 2.1403>}]
 root [{shapeType=Capsule, length=9, radius=88, shapeName=root_capsule, center=<0 0 -75.1152>, rotation=<rot 90 85.0003 0>}]
 thigh_l [{
   shapeType=Capsule
   length=33.9896
   radius=7.8703
   shapeName=thigh_l_capsule
   center=<-19.1807 -0.0357 1.443>
   rotation=<rot -0.1343 -92.8832 -90.7617>
  }]
 thigh_r [{
   shapeType=Capsule
   length=30.7079
   radius=7.7597
   shapeName=thigh_r_capsule
   center=<21.4047 0.2865 -0.1382>
   rotation=<rot 0.0653 -89.2326 -89.63>
  }]
 upperarm_r [{
   shapeType=Capsule
   length=20.5
   radius=5
   shapeName=upperarm_r_capsule
   center=<-9.5851 -0.0252 0.6454>
   rotation=<rot -2.2367 -89.9999 87.3856>
  }]
}
constraints [
 {bone1Name=Dyn_rope_bk_01, bone2Name=Base_rope_bk, swing1LimitDegrees=45, swing2LimitDegrees=33, twistLimitDegrees=5}
 {bone1Name=Dyn_rope_bk_02, bone2Name=Dyn_rope_bk_01, swing1LimitDegrees=35, swing2LimitDegrees=35, twistLimitDegrees=15}
 {bone1Name=Dyn_rope_bk_03, bone2Name=Dyn_rope_bk_02, swing1LimitDegrees=35, swing2LimitDegrees=35, twistLimitDegrees=15}
 {bone1Name=Dyn_rope_fr_03, bone2Name=Dyn_rope_fr_02, swing1LimitDegrees=25, swing2LimitDegrees=20, twistLimitDegrees=17}
 {bone1Name=Dyn_rope_fr_02, bone2Name=Dyn_rope_fr_01, swing1LimitDegrees=25, swing2LimitDegrees=25, twistLimitDegrees=15}
 {bone1Name=Dyn_rope_fr_01, bone2Name=Base_rope_fr, swing1LimitDegrees=25, swing2LimitDegrees=25, twistLimitDegrees=5}
 {bone1Name=Base_rope_fr, bone2Name=Mstr_rope, swing1LimitDegrees=33, swing2LimitDegrees=3, twistLimitDegrees=5}
 {bone1Name=Base_rope_bk, bone2Name=Mstr_rope, swing1LimitDegrees=15, swing2LimitDegrees=15, twistLimitDegrees=5}
 {
  bone1Name=Mstr_rope
  bone2Name=pelvis
  swing1LimitDegrees=45
  swing1Motion=Free
  swing2LimitDegrees=45
  swing2Motion=Free
  twistLimitDegrees=45
  twistMotion=Free
 }
 {
  bone1Name=thigh_r
  bone2Name=pelvis
  swing1LimitDegrees=45
  swing1Motion=Free
  swing2LimitDegrees=45
  swing2Motion=Free
  twistLimitDegrees=45
  twistMotion=Free
 }
 {
  bone1Name=calf_r
  bone2Name=thigh_r
  swing1LimitDegrees=45
  swing1Motion=Free
  swing2LimitDegrees=45
  swing2Motion=Free
  twistLimitDegrees=45
  twistMotion=Free
 }
 {
  bone1Name=Base_Leaves_l
  bone2Name=thigh_l
  swing1LimitDegrees=45
  swing1Motion=Free
  swing2LimitDegrees=45
  swing2Motion=Free
  twistLimitDegrees=45
  twistMotion=Free
 }
 {bone1Name=Dyn_Leaves_l, bone2Name=Base_Leaves_l, swing1LimitDegrees=25, swing2LimitDegrees=25, twistLimitDegrees=10}
 {
  bone1Name=thigh_l
  bone2Name=pelvis
  swing1LimitDegrees=45
  swing1Motion=Free
  swing2LimitDegrees=45
  swing2Motion=Free
  twistLimitDegrees=45
  twistMotion=Free
 }
 {
  bone1Name=pelvis
  bone2Name=root
  swing1LimitDegrees=45
  swing1Motion=Free
  swing2LimitDegrees=45
  swing2Motion=Free
  twistLimitDegrees=45
  twistMotion=Free
 }
 {
  bone1Name=Dyn_rope_bk_03
  bone2Name=Dyn_rope_fr_03
  swing1LimitDegrees=45
  swing1Motion=Free
  swing2LimitDegrees=45
  swing2Motion=Free
  twistLimitDegrees=45
  twistMotion=Free
 }
 {bone1Name=Dyn_rope_bk_04, bone2Name=Dyn_rope_bk_03, swing1LimitDegrees=25, swing2LimitDegrees=25, twistLimitDegrees=7}
 {bone1Name=Dyn_rope_fr_04, bone2Name=Dyn_rope_fr_03, swing1LimitDegrees=25, swing2LimitDegrees=25, twistLimitDegrees=7}
 {
  bone1Name=Dyn_rope_bk_04
  bone2Name=Dyn_rope_fr_04
  swing1LimitDegrees=45
  swing1Motion=Free
  swing2LimitDegrees=45
  swing2Motion=Free
  twistLimitDegrees=45
  twistMotion=Free
 }
 {
  bone1Name=upperarm_r
  bone2Name=pelvis
  swing1LimitDegrees=45
  swing1Motion=Free
  swing2LimitDegrees=45
  swing2Motion=Free
  twistLimitDegrees=45
  twistMotion=Free
 }
 {
  bone1Name=lowerarm_r
  bone2Name=upperarm_r
  swing1LimitDegrees=45
  swing1Motion=Free
  swing2LimitDegrees=45
  swing2Motion=Free
  twistLimitDegrees=45
  twistMotion=Free
 }
]
