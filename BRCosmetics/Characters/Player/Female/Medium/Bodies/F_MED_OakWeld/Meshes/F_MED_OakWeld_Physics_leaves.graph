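# Physics asset for the OakWeld female medium body cosmetic, named for its leaves and covering only part of the upper body.
# Three capsule bodies wrap spine_03, spine_05 and the right upper arm; no leaf bones or left-side bodies are present.
# Two constraints chain upperarm_r to spine_05 to spine_03, with swing and twist all free, so their angle limits are not enforced.
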PhysicsAsset /BRCosmetics/Characters/Player/Female/Medium/Bodies/F_MED_OakWeld/Meshes/F_MED_OakWeld_Physics_leaves
bodies {
 spine_03 [{shapeType=Capsule, length=13.5, radius=10, shapeName=spine_03_capsule, center=<-5.6463 1.4546 0>, rotation=<rot 0 -90 0>}]
 spine_05 [{
   shapeType=Capsule
   length=16.0065
   radius=10.0016
   shapeName=spine_05_capsule
   center=<-3.5039 1.4914 0.9582>
   rotation=<rot 0 1.7434 -174.7364>
  }]
 upperarm_r [{
   shapeType=Capsule
   length=23.4219
   radius=5.3532
   shapeName=upperarm_r_capsule
   center=<-17.3097 1.1295 -0.0613>
   rotation=<rot -2.0024 -91.4678 -91.0107>
  }]
}
constraints [
 {
  bone1Name=upperarm_r
  bone2Name=spine_05
  swing1LimitDegrees=45
  swing1Motion=Free
  swing2LimitDegrees=45
  swing2Motion=Free
  twistLimitDegrees=45
  twistMotion=Free
 }
 {
  bone1Name=spine_05
  bone2Name=spine_03
  swing1LimitDegrees=45
  swing1Motion=Free
  swing2LimitDegrees=45
  swing2Motion=Free
  twistLimitDegrees=45
  twistMotion=Free
 }
]
