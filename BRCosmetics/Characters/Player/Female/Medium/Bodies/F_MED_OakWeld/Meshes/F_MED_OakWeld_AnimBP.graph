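# Secondary-motion Anim Blueprint for the female OakWeld character body, animating leaves, boots and rope bones.
# A pose copied from the attached parent passes through Control Rig, RigidBody and Anim Dynamics, offsets attach_backpack, then is cached.
# When backbling, capes or disabled physics set a flag, constraints pin back leaves and rope dynamics run before output.
# Events track skydiving, backpack part and front-end simulation space; several physics and acorn nodes are disconnected.
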AnimBlueprint /BRCosmetics/Characters/Player/Female/Medium/Bodies/F_MED_OakWeld/Meshes/F_MED_OakWeld_AnimBP
target_skeleton=&/F_MED_OakWeld_Skeleton
graphs [
 AnimGraph {nodes [
   n1 = Anim Dynamics - Bone: Dyn_rope_fr_01!(SimulationSpace:AnimPhysSimSpaceType Enum=World, ComponentPose, Alpha=1, ExternalForce="0,0,0") -> (Pose) {
    angularDampingOverride=0.9
    angularSpringConstant=200
    bAngularSpring=true
    bOverrideAngularDamping=true
    bUseGravityOverride=true
    linearDampingOverride=0.9999
    simulationSpace=World
    alphaBoolBlend {blendOption=Linear}
    boundBone {boneName=Dyn_rope_fr_01}
    componentAppliedLinearAccClamp=<1000 1000 1000>
    componentLinearAccScale=<0.2 0.2 0.2>
    componentLinearVelScale=<0.2 0.2 0.2>
    gravityOverride {z=-980}
    physicsBodyDefinitions [{
      boundBone {boneName=Dyn_rope_fr_01}
      boxExtents=<13 13 13>
      constraintSetup {angularLimitsMax=<3 30 10>, angularLimitsMin=<-3 -1 -10>, angularTarget=<1 0 0>}
      localJointOffset=<0.11 0 0>
     }]
    simSpaceSettings {simSpaceAngularAlpha=1}
   }
   n2 = Anim Dynamics - Bone: Dyn_rope_fr_02!(SimulationSpace:AnimPhysSimSpaceType Enum=World, ComponentPose=@n1, Alpha=1, ExternalForce="0,0,0") -> (Pose) {
    angularDampingOverride=0.9
    angularSpringConstant=200
    bAngularSpring=true
    bOverrideAngularDamping=true
    bUseGravityOverride=true
    linearDampingOverride=0.9999
    simulationSpace=World
    alphaBoolBlend {blendOption=Linear}
    boundBone {boneName=Dyn_rope_fr_02}
    componentAppliedLinearAccClamp=<1000 1000 1000>
    componentLinearAccScale=<0.2 0.2 0.2>
    componentLinearVelScale=<0.2 0.2 0.2>
    gravityOverride {z=-980}
    physicsBodyDefinitions [{
      boundBone {boneName=Dyn_rope_fr_02}
      boxExtents=<13 13 13>
      constraintSetup {angularLimitsMax=<3 30 10>, angularLimitsMin=<-3 -1 -10>, angularTarget=<1 0 0>}
      localJointOffset=<0.11 0 0>
     }]
    simSpaceSettings {simSpaceAngularAlpha=1}
   }
   Anim Dynamics - Bone: Dyn_rope_fr_03!(SimulationSpace:AnimPhysSimSpaceType Enum=World, ComponentPose=@n2, Alpha=1, ExternalForce="0,0,0") -> (Pose) {
    angularDampingOverride=0.9
    angularSpringConstant=200
    bAngularSpring=true
    bOverrideAngularDamping=true
    bUseGravityOverride=true
    linearDampingOverride=0.9999
    simulationSpace=World
    alphaBoolBlend {blendOption=Linear}
    boundBone {boneName=Dyn_rope_fr_03}
    componentAppliedLinearAccClamp=<1000 1000 1000>
    componentLinearAccScale=<0.2 0.2 0.2>
    componentLinearVelScale=<0.2 0.2 0.2>
    gravityOverride {z=-980}
    physicsBodyDefinitions [{
      boundBone {boneName=Dyn_rope_fr_03}
      boxExtents=<13 13 13>
      constraintSetup {angularLimitsMax=<3 30 10>, angularLimitsMin=<-3 -1 -10>, angularTarget=<1 0 0>}
      localJointOffset=<0.11 0 0>
     }]
    simSpaceSettings {simSpaceAngularAlpha=1}
   }
   n3 = RigidBody!(ComponentPose=Local To Component(), Alpha, OverrideWorldGravity:Vector="0,0,-980", ExternalForce) -> (Pose) {
    bClampLinearTranslationLimitToRefPose=true
    bOverrideWorldGravity=true
    simulationSpace=BaseBoneSpace
    alphaBoolBlend {blendOption=Linear}
    baseBoneRef {boneName=spine_05}
    overrideWorldGravity {z=-980}
    simSpaceSettings {
     dampingAlpha=0
     maxAngularAcceleration=100
     maxAngularVelocity=10
     maxLinearAcceleration=500
     maxLinearVelocity=250
     worldAlpha=1
     externalLinearDragV=<0.3 0.1 0.1>
    }
   }
   Component To Local(ComponentPose=@n3)
   n4 = "Constraint - Dyn_acorn1 (PPPPPPPP)"!(
    ComponentPose
    Alpha=1
    ConstraintWeights_0=0.0012
    ConstraintWeights_1:float=0.0125
    ConstraintWeights_2:float=0.1101
    ConstraintWeights_3:float=0.0002
    ConstraintWeights_4:float=0.0007
    ConstraintWeights_5:float=0.0002
    ConstraintWeights_6:float=0.0014
    ConstraintWeights_7:float=0.8737
   ) -> (Pose) {
    boneToModify {boneName=Dyn_acorn1}
    constraintSetup [
     {targetBone {boneName=spine_04}}
     {targetBone {boneName=spine_02}}
     {targetBone {boneName=pelvis}}
     {targetBone {boneName=thigh_twist_01_l}}
     {targetBone {boneName=thigh_twist_01_r}}
     {targetBone {boneName=deform_groin_l}}
     {targetBone {boneName=deform_groin_r}}
     {targetBone {boneName=Base_acorn}}
    ]
    constraintWeights [1,1,1,1,1,1,1,1]
   }
   "Constraint - Dyn_acorn (PPPPPPPP)"!(
    ComponentPose=@n4
    Alpha=1
    ConstraintWeights_0=0
    ConstraintWeights_1:float=0.0003
    ConstraintWeights_2:float=0.0316
    ConstraintWeights_3:float=0
    ConstraintWeights_4:float=0
    ConstraintWeights_5:float=0
    ConstraintWeights_6:float=0
    ConstraintWeights_7:float=0.968
   ) -> (Pose) {
    boneToModify {boneName=Dyn_acorn}
    constraintSetup [
     {targetBone {boneName=spine_04}}
     {targetBone {boneName=spine_02}}
     {targetBone {boneName=pelvis}}
     {targetBone {boneName=thigh_twist_01_l}}
     {targetBone {boneName=thigh_twist_01_r}}
     {targetBone {boneName=deform_groin_l}}
     {targetBone {boneName=deform_groin_r}}
     {targetBone {boneName=Base_acorn}}
    ]
    constraintWeights [1,1,1,1,1,1,1,1]
   }
   "Constraint - Dyn_acorn2 (PPPPPPPPP)"!(
    ComponentPose
    Alpha=1
    ConstraintWeights_0=0
    ConstraintWeights_1:float=0
    ConstraintWeights_2:float=0.0007
    ConstraintWeights_3:float=0.0001
    ConstraintWeights_4:float=0.0839
    ConstraintWeights_5:float=0.0001
    ConstraintWeights_6:float=0
    ConstraintWeights_7:float=0.9152
    ConstraintWeights_8:float=0
   ) -> (Pose) {
    boneToModify {boneName=Dyn_acorn2}
    constraintSetup [
     {targetBone {boneName=spine_04}}
     {targetBone {boneName=spine_03}}
     {targetBone {boneName=spine_02}}
     {targetBone {boneName=spine_01}}
     {targetBone {boneName=pelvis}}
     {targetBone {boneName=thigh_twist_01_l}}
     {targetBone {boneName=deform_groin_l}}
     {targetBone {boneName=Base_acorn}}
     {targetBone {boneName=Dyn_rope_fr_03}}
    ]
    constraintWeights [1,1,1,1,1,1,1,1,1]
   }
   RigidBody!(ComponentPose, Alpha, ExternalForce) -> (Pose) {
    bOverrideWorldGravity=true
    overridePhysicsAsset=&/F_MED_OakWeld_Physics_leaves
    simulationSpace=BaseBoneSpace
    baseBoneRef {boneName=spine_05}
    overrideWorldGravity {z=-980}
    simSpaceSettings {
     maxAngularAcceleration=250
     maxAngularVelocity=250
     maxLinearAcceleration=1300
     maxLinearVelocity=250
     worldAlpha=1
     externalLinearDragV=<3 0.5 0.5>
    }
   }
   n5 = "Constraint - Dyn_Leaves_bk_l_02 (P)"(ComponentPose=Local To Component(LocalPose=Use cached pose 'Common Cache'()), Alpha=1, ConstraintWeights_0=1) {boneToModify {boneName=Dyn_Leaves_bk_l_02}, constraintSetup [{targetBone {boneName=spine_05}}], constraintWeights [1]}
   n6 = "Constraint - Dyn_Leaves_bk_l_01 (P)"(ComponentPose=@n5, Alpha=1, ConstraintWeights_0=1) {boneToModify {boneName=Dyn_Leaves_bk_l_01}, constraintSetup [{targetBone {boneName=clavicle_l}}], constraintWeights [1]}
   n7 = "Constraint - Dyn_Leaves_bk_l_03 (P)"(ComponentPose=@n6, Alpha=1, ConstraintWeights_0=1) {boneToModify {boneName=Dyn_Leaves_bk_l_03}, constraintSetup [{targetBone {boneName=spine_04}}], constraintWeights [1]}
   n8 = "Constraint - Base_Leaves_bk_l_Armpit_dwn (P)"(ComponentPose=@n7, Alpha=1, ConstraintWeights_0=1) {boneToModify {boneName=Base_Leaves_bk_l_Armpit_dwn}, constraintSetup [{targetBone {boneName=spine_03}}], constraintWeights [1]}
   n9 = "Constraint - Dyn_Leaves_bk_r_01 (P)"(ComponentPose=@n8, Alpha=1, ConstraintWeights_0=1) {boneToModify {boneName=Dyn_Leaves_bk_r_01}, constraintSetup [{targetBone {boneName=spine_05}}], constraintWeights [1]}
   n10 = "Constraint - Dyn_Leaves_bk_r_02 (P)"(ComponentPose=@n9, Alpha=1, ConstraintWeights_0=1) {boneToModify {boneName=Dyn_Leaves_bk_r_02}, constraintSetup [{targetBone {boneName=spine_05}}], constraintWeights [1]}
   n11 = "Constraint - Dyn_Leaves_bk_r_Armpit_dwn_02 (P)"(ComponentPose=@n10, Alpha=1, ConstraintWeights_0=1) {boneToModify {boneName=Dyn_Leaves_bk_r_Armpit_dwn_02}, constraintSetup [{targetBone {boneName=spine_04}}], constraintWeights [1]}
   n12 = "Constraint - Dyn_Leaves_bk_r_Armpit_dwn_03 (P)"(ComponentPose=@n11, Alpha=1, ConstraintWeights_0=1) {boneToModify {boneName=Dyn_Leaves_bk_r_Armpit_dwn_03}, constraintSetup [{targetBone {boneName=spine_05}}], constraintWeights [1]}
   n13 = Anim Dynamics - Bone: Dyn_rope_fr_03(ComponentPose=@n12, Alpha=1, ExternalForce="0,0,0") {
    angularDampingOverride=0.9999
    angularSpringConstant=500
    bAngularSpring=true
    bOverrideAngularDamping=true
    bUseGravityOverride=true
    bUsePlanarLimit=false
    bUseSphericalLimits=true
    linearDampingOverride=0.9999
    linearSpringConstant=500
    simulationSpace=RootRelative
    alphaBoolBlend {blendOption=Linear}
    boundBone {boneName=Dyn_rope_fr_03}
    componentAppliedLinearAccClamp=<5000 5000 5000>
    componentLinearAccScale=<0.5 0.5 0.5>
    componentLinearVelScale=<0.5 0.5 0.5>
    gravityOverride {z=-980}
    physicsBodyDefinitions [{
      collisionType=OuterSphere
      sphereCollisionRadius=1.3
      boundBone {boneName=Dyn_rope_fr_03}
      boxExtents=<9 9 9>
      constraintSetup {angularLimitsMax=<5 20 15>, angularLimitsMin=<-5 -20 -15>, angularTarget=<1 0 0>}
      localJointOffset=<1 0 0>
     }]
    simSpaceSettings {simSpaceAngularAlpha=1}
    sphericalLimits [{limitRadius=6, drivingBone {boneName=thigh_r}, sphereLocalOffset=<17.9879 -0.5652 0>}]
   }
   n14 = Anim Dynamics - Bone: Dyn_rope_fr_04!(SimulationSpace:AnimPhysSimSpaceType Enum, ComponentPose=@n13, Alpha=1, ExternalForce="0,0,0") -> (Pose) {
    angularDampingOverride=0.9999
    angularSpringConstant=500
    bAngularSpring=true
    bOverrideAngularDamping=true
    bUseGravityOverride=true
    bUsePlanarLimit=false
    bUseSphericalLimits=true
    linearDampingOverride=0.9999
    linearSpringConstant=500
    simulationSpace=RootRelative
    alphaBoolBlend {blendOption=Linear}
    boundBone {boneName=Dyn_rope_fr_04}
    componentAppliedLinearAccClamp=<5000 5000 5000>
    componentLinearAccScale=<0.5 0.5 0.5>
    componentLinearVelScale=<0.5 0.5 0.5>
    gravityOverride {z=-980}
    physicsBodyDefinitions [{
      collisionType=OuterSphere
      sphereCollisionRadius=1.3
      boundBone {boneName=Dyn_rope_fr_04}
      boxExtents=<9 9 9>
      constraintSetup {angularLimitsMax=<5 20 15>, angularLimitsMin=<-5 -20 -15>, angularTarget=<1 0 0>}
      localJointOffset=<1 0 0>
     }]
    simSpaceSettings {simSpaceAngularAlpha=1}
    sphericalLimits [{limitRadius=6, drivingBone {boneName=thigh_r}, sphereLocalOffset=<27.8321 1.8977 -1.1762>}]
   }
   n15 = Anim Dynamics - Bone: Dyn_rope_bk_03!(SimulationSpace:AnimPhysSimSpaceType Enum, ComponentPose=@n14, Alpha=1, ExternalForce="0,0,0") -> (Pose) {
    angularDampingOverride=0.9999
    angularSpringConstant=500
    bAngularSpring=true
    bOverrideAngularDamping=true
    bUseGravityOverride=true
    bUsePlanarLimit=false
    bUseSphericalLimits=true
    linearDampingOverride=0.9999
    linearSpringConstant=500
    simulationSpace=RootRelative
    alphaBoolBlend {blendOption=Linear}
    boundBone {boneName=Dyn_rope_bk_03}
    componentAppliedLinearAccClamp=<5000 5000 5000>
    componentLinearAccScale=<0.5 0.5 0.5>
    componentLinearVelScale=<0.5 0.5 0.5>
    gravityOverride {z=-980}
    physicsBodyDefinitions [{
      collisionType=OuterSphere
      sphereCollisionRadius=1.3
      boundBone {boneName=Dyn_rope_bk_03}
      boxExtents=<9 9 9>
      constraintSetup {angularLimitsMax=<5 20 15>, angularLimitsMin=<-5 -20 -15>, angularTarget=<1 0 0>}
      localJointOffset=<1 0 0>
     }]
    simSpaceSettings {simSpaceAngularAlpha=1}
    sphericalLimits [{limitRadius=6, drivingBone {boneName=thigh_r}, sphereLocalOffset=<17.9879 -0.5652 0>}]
   }
   n16 = Anim Dynamics - Bone: Dyn_rope_bk_04!(SimulationSpace:AnimPhysSimSpaceType Enum, ComponentPose=@n15, Alpha=1, ExternalForce="0,0,0") -> (Pose) {
    angularDampingOverride=0.9999
    angularSpringConstant=500
    bAngularSpring=true
    bOverrideAngularDamping=true
    bUseGravityOverride=true
    bUsePlanarLimit=false
    bUseSphericalLimits=true
    linearDampingOverride=0.9999
    linearSpringConstant=500
    simulationSpace=RootRelative
    alphaBoolBlend {blendOption=Linear}
    boundBone {boneName=Dyn_rope_bk_04}
    componentAppliedLinearAccClamp=<5000 5000 5000>
    componentLinearAccScale=<0.5 0.5 0.5>
    componentLinearVelScale=<0.5 0.5 0.5>
    gravityOverride {z=-980}
    physicsBodyDefinitions [{
      collisionType=OuterSphere
      sphereCollisionRadius=1.3
      boundBone {boneName=Dyn_rope_bk_04}
      boxExtents=<9 9 9>
      constraintSetup {angularLimitsMax=<5 20 15>, angularLimitsMin=<-5 -20 -15>, angularTarget=<1 0 0>}
      localJointOffset=<1 0 0>
     }]
    simSpaceSettings {simSpaceAngularAlpha=1}
    sphericalLimits [{limitRadius=6.2, drivingBone {boneName=thigh_r}, sphereLocalOffset=<37.1883 0.572 -1.1762>}]
   }
   Output Pose(Result=Blend Poses by bool(
     BlendPose_0=Component To Local(ComponentPose=@n16)
     BlendPose_1=Use cached pose 'Common Cache'()
     bActiveValue=Get is Wearing Backbling or Cape() -> (is Wearing Backbling or Cape:bool)
    )) {layerGroup=None}
   n17 = Control Rig!(
    Source=Copy Pose From Mesh() {bUseAttachedParent=true}
    Alpha
    ISRBANon:bool=Get bRigidBodyEnabled()
    IsCapeOrBackpack:bool=Get is Wearing Backbling or Cape()
    IsPlayingEmote:bool=Get bIsPlayingEmote()
    IsSkydive:double=Get SkydiveFloat() -> (SkydiveFloat:double)
   ) -> (Pose) {controlRigAssetReference {blueprintRigClass=&/F_MED_OakWeld_CtrlRig.F_MED_OakWeld_CtrlRig_C}}
   n18 = RigidBody(ComponentPose=Local To Component(LocalPose=@n17)) {
    bOverrideWorldGravity=true
    overridePhysicsAsset=&/F_MED_OakWeld_Physics
    simulationSpace=BaseBoneSpace
    baseBoneRef {boneName=pelvis}
    overrideWorldGravity {z=-980}
    simSpaceSettings {
     dampingAlpha=0.7
     maxAngularAcceleration=250
     maxAngularVelocity=100
     maxLinearAcceleration=900
     maxLinearVelocity=250
     worldAlpha=1
     externalLinearDragV {x=0.3}
    }
   }
   n19 = Anim Dynamics - Bone: Dyn_hand_leaf_l01!(SimulationSpace:AnimPhysSimSpaceType Enum, ComponentPose=@n18, Alpha=1, ExternalForce="0,0,0") -> (Pose) {
    angularDampingOverride=0.9999
    angularSpringConstant=500
    bAngularSpring=true
    bOverrideAngularDamping=true
    bUseGravityOverride=true
    bUsePlanarLimit=false
    bUseSphericalLimits=true
    linearDampingOverride=0.9999
    linearSpringConstant=500
    simulationSpace=RootRelative
    alphaBoolBlend {blendOption=Linear}
    boundBone {boneName=Dyn_hand_leaf_l01}
    componentAppliedLinearAccClamp=<5000 5000 5000>
    componentLinearAccScale=<0.5 0.5 0.5>
    componentLinearVelScale=<0.5 0.5 0.5>
    gravityOverride {z=-980}
    physicsBodyDefinitions [{
      sphereCollisionRadius=1.3
      boundBone {boneName=Dyn_hand_leaf_l01}
      boxExtents=<13 13 13>
      constraintSetup {angularLimitsMax=<2 10 5>, angularLimitsMin=<-2 -10 -5>, angularTarget=<1 0 0>}
      localJointOffset=<1 0 0>
     }]
    simSpaceSettings {simSpaceAngularAlpha=1}
    sphericalLimits [{limitRadius=3, drivingBone {boneName=neck_01}, sphereLocalOffset=<-6.7194 1.1698 4.9371>}]
   }
   n20 = Anim Dynamics - Bone: Dyn_hand_leaf_r01!(SimulationSpace:AnimPhysSimSpaceType Enum, ComponentPose=@n19, Alpha=1, ExternalForce="0,0,0") -> (Pose) {
    angularDampingOverride=0.9999
    angularSpringConstant=500
    bAngularSpring=true
    bOverrideAngularDamping=true
    bUseGravityOverride=true
    bUsePlanarLimit=false
    bUseSphericalLimits=true
    linearDampingOverride=0.9999
    linearSpringConstant=500
    simulationSpace=RootRelative
    alphaBoolBlend {blendOption=Linear}
    boundBone {boneName=Dyn_hand_leaf_r01}
    componentAppliedLinearAccClamp=<5000 5000 5000>
    componentLinearAccScale=<0.5 0.5 0.5>
    componentLinearVelScale=<0.5 0.5 0.5>
    gravityOverride {z=-980}
    physicsBodyDefinitions [{
      sphereCollisionRadius=1.3
      boundBone {boneName=Dyn_hand_leaf_r01}
      boxExtents=<13 13 13>
      constraintSetup {angularLimitsMax=<2 2 5>, angularLimitsMin=<-2 -10 -5>, angularTarget=<1 0 0>}
      localJointOffset=<1 0 0>
     }]
    simSpaceSettings {simSpaceAngularAlpha=1}
    sphericalLimits [{limitRadius=3, drivingBone {boneName=neck_01}, sphereLocalOffset=<-6.7194 1.1698 4.9371>}]
   }
   n21 = Anim Dynamics - Bone: Dyn_boots_l!(SimulationSpace:AnimPhysSimSpaceType Enum, ComponentPose=@n20, Alpha=1, ExternalForce="0,0,0") -> (Pose) {
    angularDampingOverride=0.9999
    angularSpringConstant=500
    bAngularSpring=true
    bOverrideAngularDamping=true
    bUseGravityOverride=true
    bUsePlanarLimit=false
    bUseSphericalLimits=true
    linearDampingOverride=0.9999
    linearSpringConstant=500
    simulationSpace=RootRelative
    alphaBoolBlend {blendOption=Linear}
    boundBone {boneName=Dyn_boots_l}
    componentAppliedLinearAccClamp=<5000 5000 5000>
    componentLinearAccScale=<0.5 0.5 0.5>
    componentLinearVelScale=<0.5 0.5 0.5>
    gravityOverride {z=-980}
    physicsBodyDefinitions [{
      sphereCollisionRadius=1.3
      boundBone {boneName=Dyn_boots_l}
      boxExtents=<13 13 13>
      constraintSetup {angularLimitsMax=<2 10 5>, angularLimitsMin=<-2 -7 -5>, angularTarget=<1 0 0>}
      localJointOffset=<1 0 0>
     }]
    simSpaceSettings {simSpaceAngularAlpha=1}
    sphericalLimits [{limitRadius=3, drivingBone {boneName=neck_01}, sphereLocalOffset=<-6.7194 1.1698 4.9371>}]
   }
   n22 = Anim Dynamics - Bone: Dyn_boots_r!(SimulationSpace:AnimPhysSimSpaceType Enum, ComponentPose=@n21, Alpha=1, ExternalForce="0,0,0") -> (Pose) {
    angularDampingOverride=0.9999
    angularSpringConstant=500
    bAngularSpring=true
    bOverrideAngularDamping=true
    bUseGravityOverride=true
    bUsePlanarLimit=false
    bUseSphericalLimits=true
    linearDampingOverride=0.9999
    linearSpringConstant=500
    simulationSpace=RootRelative
    alphaBoolBlend {blendOption=Linear}
    boundBone {boneName=Dyn_boots_r}
    componentAppliedLinearAccClamp=<5000 5000 5000>
    componentLinearAccScale=<0.5 0.5 0.5>
    componentLinearVelScale=<0.5 0.5 0.5>
    gravityOverride {z=-980}
    physicsBodyDefinitions [{
      sphereCollisionRadius=1.3
      boundBone {boneName=Dyn_boots_r}
      boxExtents=<13 13 13>
      constraintSetup {angularLimitsMax=<2 7 5>, angularLimitsMin=<-2 -15 -5>, angularTarget=<1 0 0>}
      localJointOffset=<1 0 0>
     }]
    simSpaceSettings {simSpaceAngularAlpha=1}
    sphericalLimits [{limitRadius=3, drivingBone {boneName=neck_01}, sphereLocalOffset=<-6.7194 1.1698 4.9371>}]
   }
   n23 = Anim Dynamics - Bone: Dyn_Leaves_bk_r_03!(SimulationSpace:AnimPhysSimSpaceType Enum, ComponentPose=@n22, Alpha=1, ExternalForce="0,0,0") -> (Pose) {
    angularDampingOverride=1
    angularSpringConstant=500
    bAngularSpring=true
    bOverrideAngularDamping=true
    bUseGravityOverride=true
    bUsePlanarLimit=false
    bUseSphericalLimits=true
    linearDampingOverride=0.9999
    linearSpringConstant=500
    simulationSpace=RootRelative
    alphaBoolBlend {blendOption=Linear}
    boundBone {boneName=Dyn_Leaves_bk_r_03}
    componentAppliedLinearAccClamp=<1000 1000 1000>
    componentLinearAccScale {z=0.5}
    componentLinearVelScale {z=0.5}
    gravityOverride {z=-980}
    physicsBodyDefinitions [{
      collisionType=CustomSphere
      sphereCollisionRadius=1.4
      boundBone {boneName=Dyn_Leaves_bk_r_03}
      boxExtents=<13 13 13>
      constraintSetup {angularLimitsMax=<0 35 2>, angularLimitsMin=<0 -2 -2>, angularTarget=<1 0 0>}
      localJointOffset=<1 0 0>
     }]
    simSpaceSettings {maxAngularAcceleration=7000, maxAngularVelocity=7000, simSpaceAngularAlpha=1}
    sphericalLimits [{limitRadius=12, drivingBone {boneName=spine_02}, sphereLocalOffset=<4.2011 0.0144 0>}]
   }
   n24 = Anim Dynamics - Bone: Dyn_Leaves_bk_r_02!(SimulationSpace:AnimPhysSimSpaceType Enum, ComponentPose=@n23, Alpha=1, ExternalForce="0,0,0") -> (Pose) {
    angularDampingOverride=1
    angularSpringConstant=500
    bAngularSpring=true
    bOverrideAngularDamping=true
    bUseGravityOverride=true
    bUsePlanarLimit=false
    bUseSphericalLimits=true
    linearDampingOverride=0.9999
    linearSpringConstant=500
    simulationSpace=RootRelative
    alphaBoolBlend {blendOption=Linear}
    boundBone {boneName=Dyn_Leaves_bk_r_02}
    componentAppliedLinearAccClamp=<1000 1000 1000>
    componentLinearAccScale {z=0.5}
    componentLinearVelScale {z=0.5}
    gravityOverride {z=-980}
    physicsBodyDefinitions [{
      collisionType=CustomSphere
      sphereCollisionRadius=1
      boundBone {boneName=Dyn_Leaves_bk_r_02}
      boxExtents=<13 13 13>
      constraintSetup {angularLimitsMax=<0 35 2>, angularLimitsMin=<0 -2 -2>, angularTarget=<1 0 0>}
      localJointOffset=<1 0 0>
     }]
    simSpaceSettings {maxAngularAcceleration=7000, maxAngularVelocity=7000, simSpaceAngularAlpha=1}
    sphericalLimits [{limitRadius=10.5, drivingBone {boneName=spine_05}, sphereLocalOffset=<-4.1885 1.2321 0>}]
   }
   n25 = Anim Dynamics - Bone: Dyn_Leaves_bk_r_Armpit_dwn_02!(SimulationSpace:AnimPhysSimSpaceType Enum, ComponentPose=@n24, Alpha=1, ExternalForce="0,0,0") -> (Pose) {
    angularDampingOverride=1
    angularSpringConstant=500
    bAngularSpring=true
    bOverrideAngularDamping=true
    bUseGravityOverride=true
    bUsePlanarLimit=false
    bUseSphericalLimits=true
    linearDampingOverride=0.9999
    linearSpringConstant=500
    simulationSpace=RootRelative
    alphaBoolBlend {blendOption=Linear}
    boundBone {boneName=Dyn_Leaves_bk_r_Armpit_dwn_02}
    componentAppliedLinearAccClamp=<1000 1000 1000>
    componentLinearAccScale {z=0.5}
    componentLinearVelScale {z=0.5}
    gravityOverride {z=-980}
    physicsBodyDefinitions [{
      sphereCollisionRadius=1.3
      boundBone {boneName=Dyn_Leaves_bk_r_Armpit_dwn_02}
      boxExtents=<13 13 13>
      constraintSetup {angularLimitsMax=<0 35 2>, angularLimitsMin=<0 -2 -2>, angularTarget=<1 0 0>}
      localJointOffset=<1 0 0>
     }]
    simSpaceSettings {maxAngularAcceleration=7000, maxAngularVelocity=7000, simSpaceAngularAlpha=1}
    sphericalLimits [{limitRadius=3, drivingBone {boneName=neck_01}, sphereLocalOffset=<-6.7194 1.1698 4.9371>}]
   }
   n26 = Anim Dynamics - Bone: Dyn_Leaves_bk_r_Armpit_dwn_03!(SimulationSpace:AnimPhysSimSpaceType Enum, ComponentPose=@n25, Alpha=1, ExternalForce="0,0,0") -> (Pose) {
    angularDampingOverride=1
    angularSpringConstant=500
    bAngularSpring=true
    bOverrideAngularDamping=true
    bUseGravityOverride=true
    bUsePlanarLimit=false
    bUseSphericalLimits=true
    linearDampingOverride=0.9999
    linearSpringConstant=500
    simulationSpace=RootRelative
    alphaBoolBlend {blendOption=Linear}
    boundBone {boneName=Dyn_Leaves_bk_r_Armpit_dwn_03}
    componentAppliedLinearAccClamp=<1000 1000 1000>
    componentLinearAccScale {z=0.5}
    componentLinearVelScale {z=0.5}
    gravityOverride {z=-980}
    physicsBodyDefinitions [{
      sphereCollisionRadius=1.3
      boundBone {boneName=Dyn_Leaves_bk_r_Armpit_dwn_03}
      boxExtents=<13 13 13>
      constraintSetup {angularLimitsMax=<0 35 2>, angularLimitsMin=<0 -2 -2>, angularTarget=<1 0 0>}
      localJointOffset=<1 0 0>
     }]
    simSpaceSettings {maxAngularAcceleration=7000, maxAngularVelocity=7000, simSpaceAngularAlpha=1}
    sphericalLimits [{limitRadius=3, drivingBone {boneName=neck_01}, sphereLocalOffset=<-6.7194 1.1698 4.9371>}]
   }
   n27 = Anim Dynamics - Bone: Dyn_Leaves_bk_r_Armpit_up!(SimulationSpace:AnimPhysSimSpaceType Enum, ComponentPose=@n26, Alpha=1, ExternalForce="0,0,0") -> (Pose) {
    angularDampingOverride=1
    angularSpringConstant=500
    bAngularSpring=true
    bOverrideAngularDamping=true
    bUseGravityOverride=true
    bUsePlanarLimit=false
    bUseSphericalLimits=true
    linearDampingOverride=0.9999
    linearSpringConstant=500
    simulationSpace=RootRelative
    alphaBoolBlend {blendOption=Linear}
    boundBone {boneName=Dyn_Leaves_bk_r_Armpit_up}
    componentAppliedLinearAccClamp=<1000 1000 1000>
    componentLinearAccScale {z=0.5}
    componentLinearVelScale {z=0.5}
    gravityOverride {z=-980}
    physicsBodyDefinitions [{
      sphereCollisionRadius=1.3
      boundBone {boneName=Dyn_Leaves_bk_r_Armpit_up}
      boxExtents=<13 13 13>
      constraintSetup {angularLimitsMax=<0 35 2>, angularLimitsMin=<0 -2 -2>, angularTarget=<1 0 0>}
      localJointOffset=<0.5 0 0>
     }]
    simSpaceSettings {maxAngularAcceleration=7000, maxAngularVelocity=7000, simSpaceAngularAlpha=1}
    sphericalLimits [{limitRadius=3, drivingBone {boneName=neck_01}, sphereLocalOffset=<-6.7194 1.1698 4.9371>}]
   }
   n28 = Anim Dynamics - Bone: Dyn_Leaves_bk_l_02!(SimulationSpace:AnimPhysSimSpaceType Enum, ComponentPose=@n27, Alpha=1, ExternalForce="0,0,0") -> (Pose) {
    angularDampingOverride=1
    angularSpringConstant=500
    bAngularSpring=true
    bOverrideAngularDamping=true
    bUseGravityOverride=true
    bUsePlanarLimit=false
    bUseSphericalLimits=true
    linearDampingOverride=0.9999
    linearSpringConstant=500
    simulationSpace=RootRelative
    alphaBoolBlend {blendOption=Linear}
    boundBone {boneName=Dyn_Leaves_bk_l_02}
    componentAppliedLinearAccClamp=<1000 1000 1000>
    componentLinearAccScale {z=0.5}
    componentLinearVelScale {z=0.5}
    gravityOverride {z=-980}
    physicsBodyDefinitions [{
      sphereCollisionRadius=1.3
      boundBone {boneName=Dyn_Leaves_bk_l_02}
      boxExtents=<13 13 13>
      constraintSetup {angularLimitsMax=<0 35 2>, angularLimitsMin=<0 -2 -2>, angularTarget=<1 0 0>}
      localJointOffset=<1 0 0>
     }]
    simSpaceSettings {maxAngularAcceleration=7000, maxAngularVelocity=7000, simSpaceAngularAlpha=1}
    sphericalLimits [{limitRadius=3, drivingBone {boneName=neck_01}, sphereLocalOffset=<-6.7194 1.1698 4.9371>}]
   }
   n29 = Anim Dynamics - Bone: Dyn_Leaves_bk_l_03!(SimulationSpace:AnimPhysSimSpaceType Enum, ComponentPose=@n28, Alpha=1, ExternalForce="0,0,0") -> (Pose) {
    angularDampingOverride=1
    angularSpringConstant=500
    bAngularSpring=true
    bOverrideAngularDamping=true
    bUseGravityOverride=true
    bUsePlanarLimit=false
    bUseSphericalLimits=true
    linearDampingOverride=0.9999
    linearSpringConstant=500
    simulationSpace=RootRelative
    alphaBoolBlend {blendOption=Linear}
    boundBone {boneName=Dyn_Leaves_bk_l_03}
    componentAppliedLinearAccClamp=<1000 1000 1000>
    componentLinearAccScale {z=0.5}
    componentLinearVelScale {z=0.5}
    gravityOverride {z=-980}
    physicsBodyDefinitions [{
      sphereCollisionRadius=1.3
      boundBone {boneName=Dyn_Leaves_bk_l_03}
      boxExtents=<13 13 13>
      constraintSetup {angularLimitsMax=<0 35 2>, angularLimitsMin=<0 -2 -2>, angularTarget=<1 0 0>}
      localJointOffset=<1 0 0>
     }]
    simSpaceSettings {maxAngularAcceleration=7000, maxAngularVelocity=7000, simSpaceAngularAlpha=1}
    sphericalLimits [{limitRadius=3, drivingBone {boneName=neck_01}, sphereLocalOffset=<-6.7194 1.1698 4.9371>}]
   }
   n30 = Anim Dynamics - Bone: Base_Leaves_bk_l_Armpit_dwn!(SimulationSpace:AnimPhysSimSpaceType Enum, ComponentPose=@n29, Alpha=1, ExternalForce="0,0,0") -> (Pose) {
    angularDampingOverride=1
    angularSpringConstant=500
    bAngularSpring=true
    bOverrideAngularDamping=true
    bUseGravityOverride=true
    bUsePlanarLimit=false
    bUseSphericalLimits=true
    linearDampingOverride=0.9999
    linearSpringConstant=500
    simulationSpace=RootRelative
    alphaBoolBlend {blendOption=Linear}
    boundBone {boneName=Base_Leaves_bk_l_Armpit_dwn}
    componentAppliedLinearAccClamp=<1000 1000 1000>
    componentLinearAccScale {z=0.5}
    componentLinearVelScale {z=0.5}
    gravityOverride {z=-980}
    physicsBodyDefinitions [{
      sphereCollisionRadius=1.3
      boundBone {boneName=Base_Leaves_bk_l_Armpit_dwn}
      boxExtents=<13 13 13>
      constraintSetup {angularLimitsMax=<0 35 2>, angularLimitsMin=<0 -2 -2>, angularTarget=<1 0 0>}
      localJointOffset=<1 0 0>
     }]
    simSpaceSettings {maxAngularAcceleration=7000, maxAngularVelocity=7000, simSpaceAngularAlpha=1}
    sphericalLimits [{limitRadius=3, drivingBone {boneName=neck_01}, sphereLocalOffset=<-6.7194 1.1698 4.9371>}]
   }
   n31 = "Transform (Modify) Bone - Bone: attach_backpack"(ComponentPose=@n30, Alpha=1, Translation="0,1.75,7", Rotation="0,0,0", Scale="1,1,1") {translationMode=BMM_Additive, translationSpace=BCS_BoneSpace, boneToModify {boneName=attach_backpack}, translation {y=1.75, z=7}}
   Save cached pose 'Common Cache'(Pose=Component To Local(ComponentPose=@n31)) {cacheName=Common Cache}
  ]}
 EventGraph {nodes [
   Event BlueprintUpdateAnimation()
   > n1 = Sequence!(execute) -> (then_0, then_1, then_2:Exec)
   > n2 = Branch(Condition=Get bIsInFrontEnd())
   > Set SimlationSpace(SimlationSpace:AnimPhysSimSpaceType Enum=World) -> (Output_Get:AnimPhysSimSpaceType Enum)
   Set SimlationSpace(execute=@n2.else, SimlationSpace=Component)
   n3 = Get bRigidBodyEnabled()
   n4 = CheckForBackbling()
   n5 = NOT Boolean(A=@n4.bIsBackblingLong)
   n6 = NOT Boolean(A=@n4.bIsBackblingACape)
   Set is Wearing Backbling or Cape(
    execute=@n1.then_1
    is Wearing Backbling or Cape:bool=OR Boolean!(
     A=AND Boolean!(A=@n4, B=@n6, C:bool=@n5) -> (ReturnValue)
     B=OR Boolean!(A=@n4.bIsBackblingACape, B=@n4.bIsBackblingLong) -> (ReturnValue)
     C:bool=NOT Boolean(A=@n3)
    ) -> (ReturnValue)
   ) -> (Output_Get:bool)
   n7 = Event OnCharacterCustomizationLoaded()
   > n8 = For Each Loop(Array:Array of Custom Character Part Object References=@n7.CustomParts) -> (Array Element:Custom Character Part Object Reference)
   > Branch(Condition="Equal (Enum)"(
     A:EFortCustomPartType Enum=GetPartType!(self:Custom Character Part Object Reference=@n8) -> (ReturnValue:EFortCustomPartType Enum=Head)
     B:EFortCustomPartType Enum=Backpack
    ))
   > n9 = Branch(Condition=Contains(SearchIn=GetObjectName(Object=@n8), Substring=OakWeld))
   > Set IsThisMyBackBling(IsThisMyBackBling:bool=true) -> (Output_Get:bool)
   Set IsThisMyBackBling(execute=@n9.else, IsThisMyBackBling=false)
   Event OnEndSkydiving()
   > Set SkydiveFloat(SkydiveFloat:double=0) -> (Output_Get:double)
   Event OnBeginSkydiving()
   > Set SkydiveFloat(SkydiveFloat=1)
  ]}
]
variables [
 is Wearing Backbling or Cape {type=boolean}
 IsThisMyBackBling {type=boolean}
 SimlationSpace {type=AnimPhysSimSpaceType}
 SkydiveFloat {type=number}
]
